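# Sound cue for the supply drop radio's flare, combining an ignition one-shot with a sustained loop.
# A mixer plays the flare start wave, weighted louder, alongside a looping flare wave.
# An enveloper ramps the loop's volume up from silence over about a second, so it fades in beneath the start sound.
# It uses the close supply drop marker flare attenuation and the Athena ambient sound class.
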SoundCue /SupplyDropRadioGameplay/Audio/Fort_SupplyDropRadio_Flare_Loop_Cue
root=Mixer_0
nodes [
 {
  id=Enveloper_0
  children [WavePlayer_0]
  properties {pitchCurve {editorCurveData {keys [{time=0}]}}, volumeCurve {editorCurveData {keys [<at 0 0>,<at 0.4952 0.4615>,<at 1 1>]}}}
 }
 {id=Mixer_0, children [WavePlayer_2,Enveloper_0], properties {inputVolume [1.5,1]}}
 {id=WavePlayer_0, properties {bLooping=true, soundWaveAssetPtr=&/SupplyDropRadio_Flare_Loop}}
 {id=WavePlayer_2, properties {soundWaveAssetPtr=&/SupplyDrop_FlareStart}}
]
properties {
 attenuationSettings=/Game/Sounds/Foley_Loot/Containers/SupplyDrops/Fort_SupplyDrop_Marker_Flare_Close_Athena_Att
 maxDistance=1500
 soundClassObject=&Fort_Ambient_Athena
 volumeMultiplier=0.351
}
