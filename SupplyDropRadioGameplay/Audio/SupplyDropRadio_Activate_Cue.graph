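# Sound cue for activating the Supply Drop Radio gadget.
# A Random node picks one of three equally weighted activation wave variations and is excluded from branch culling.
# Playback uses the small gadget attenuation asset and the gadgets/items sound class at slightly reduced volume.
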SoundCue /SupplyDropRadioGameplay/Audio/SupplyDropRadio_Activate_Cue
root=Random_0
nodes [
 {id=Random_0, children [WavePlayer_0,WavePlayer_1,WavePlayer_2], properties {bShouldExcludeFromBranchCulling=true, weights [1,1,1]}}
 {id=WavePlayer_0, properties {soundWaveAssetPtr=&/SupplyDropRadio_Activate_01}}
 {id=WavePlayer_1, properties {soundWaveAssetPtr=&/SupplyDropRadio_Activate_02}}
 {id=WavePlayer_2, properties {soundWaveAssetPtr=&/SupplyDropRadio_Activate_03}}
]
properties {
 attenuationSettings=/Game/Sounds/Attenuation/Fort_Gadget_Sml_Att
 maxDistance=1100
 soundClassObject=&Fort_Gadgets_Items
 volumeMultiplier=0.7
}
