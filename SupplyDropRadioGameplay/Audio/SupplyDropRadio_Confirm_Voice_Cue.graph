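# Sound cue for the supply drop radio's confirmation voice line.
# A single wave player plays one voice wave, with no randomization or modulation.
# It uses the close supply drop marker attenuation with a short max distance, routes to the gadgets and items sound class, and plays at reduced volume.
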SoundCue /SupplyDropRadioGameplay/Audio/SupplyDropRadio_Confirm_Voice_Cue
root=WavePlayer_2
nodes [{id=WavePlayer_2, properties {soundWaveAssetPtr=&/SupplyDropRadio_Confirm_Voice_01}}]
properties {
 attenuationSettings=/Game/Sounds/Foley_Loot/Containers/SupplyDrops/Fort_SupplyDrop_Marker_Close_Athena_Att
 maxDistance=1500
 soundClassObject=&Fort_Gadgets_Items
 volumeMultiplier=0.6
}
