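# Sound cue for the drop sound of the supply drop radio gadget.
# A single wave player plays the SupplyDropRadio_Drop wave directly as the root, with no randomization or mixing.
# It uses the shared weapon-drop attenuation, routes to the gadget items sound class, and plays at reduced volume.
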SoundCue /SupplyDropRadioGameplay/Audio/SupplyDropRadio_Drop_Cue
root=WavePlayer_0
nodes [{id=WavePlayer_0, properties {soundWaveAssetPtr=&/SupplyDropRadio_Drop}}]
properties {
 attenuationSettings=&Fort_Weapon_Drop_Att
 maxDistance=3000
 soundClassObject=&Fort_Gadgets_Items
 volumeMultiplier=0.5
}
