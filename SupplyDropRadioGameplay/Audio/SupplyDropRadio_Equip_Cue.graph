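# Sound cue for equipping the supply drop radio gadget.
# A Random node picks one of three equally weighted equip wave variations on each play.
# It plays at reduced volume in the gadget items sound class, spatialized with the weapon reload attenuation settings.
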SoundCue /SupplyDropRadioGameplay/Audio/SupplyDropRadio_Equip_Cue
root=Random_0
nodes [
 {id=Random_0, children [WavePlayer_0,WavePlayer_1,WavePlayer_2], properties {weights [1,1,1]}}
 {id=WavePlayer_0, properties {soundWaveAssetPtr=&/SupplyDropRadio_Equip_01}}
 {id=WavePlayer_1, properties {soundWaveAssetPtr=&/SupplyDropRadio_Equip_02}}
 {id=WavePlayer_2, properties {soundWaveAssetPtr=&/SupplyDropRadio_Equip_03}}
]
properties {
 attenuationSettings=&Fort_Weapon_Reload_Att
 maxDistance=2250
 soundClassObject=&Fort_Gadgets_Items
 volumeMultiplier=0.55
}
