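# Sound cue for the bounce of the supply drop radio's flare.
# A random node picks one of three bounce wave variations with equal weight.
# It plays in the ranged weapon projectile effects sound class with grenade bounce attenuation.
# A concurrency override allows at most three simultaneous instances and prevents new ones beyond that.
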SoundCue /SupplyDropRadioGameplay/Audio/SupplyDropRadio_Flare_Bounce_Cue
root=Random_4
nodes [
 {id=Random_4, children [WavePlayer_5,WavePlayer_6,WavePlayer_7], properties {weights [1,1,1]}}
 {id=WavePlayer_5, properties {soundWaveAssetPtr=&/SupplyDropRadio_Flare_Bounce_03}}
 {id=WavePlayer_6, properties {soundWaveAssetPtr=&/SupplyDropRadio_Flare_Bounce_01}}
 {id=WavePlayer_7, properties {soundWaveAssetPtr=&/SupplyDropRadio_Flare_Bounce_02}}
]
properties {
 attenuationSettings=/Game/Sounds/Attenuation/Fort_Grenade_Bounce_Att
 bOverrideConcurrency=true
 maxDistance=5500
 soundClassObject=&Fort_WeaponRanged_Projectile_Effects
 volumeMultiplier=1
 concurrencyOverrides {maxCount=3, resolutionRule=PreventNew}
}
