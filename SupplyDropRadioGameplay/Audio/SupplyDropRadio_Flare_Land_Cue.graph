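# Sound cue for the supply drop radio flare's landing sound.
# A random node picks one of three equally weighted flare-landing wave variations.
# It uses the close supply drop marker flare attenuation and the gadget items sound class.
# A concurrency override allows one instance per owner and blocks new plays while one is active.
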SoundCue /SupplyDropRadioGameplay/Audio/SupplyDropRadio_Flare_Land_Cue
root=Random_0
nodes [
 {id=Random_0, children [WavePlayer_0,WavePlayer_1,WavePlayer_2], properties {weights [1,1,1]}}
 {id=WavePlayer_0, properties {soundWaveAssetPtr=&/SupplyDropRadio_Flare_Land_01}}
 {id=WavePlayer_1, properties {soundWaveAssetPtr=&/SupplyDropRadio_Flare_Land_02}}
 {id=WavePlayer_2, properties {soundWaveAssetPtr=&/SupplyDropRadio_Flare_Land_03}}
]
properties {
 attenuationSettings=/Game/Sounds/Foley_Loot/Containers/SupplyDrops/Fort_SupplyDrop_Marker_Flare_Close_Athena_Att
 bOverrideConcurrency=true
 maxDistance=1500
 soundClassObject=&Fort_Gadgets_Items
 volumeMultiplier=1
 concurrencyOverrides {bLimitToOwner=true, maxCount=1, resolutionRule=PreventNew}
}
