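# Sound cue that plays the landing sound for a supply drop radio gadget.
# A single wave player feeds a modulator that randomly lowers the pitch on each play for slight variation.
# The cue plays at reduced volume in the gadget items sound class and uses shared weapon-drop attenuation for distance falloff.
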SoundCue /SupplyDropRadioGameplay/Audio/SupplyDropRadio_Landed_Cue
root=Modulator_0
nodes [
 {id=Modulator_0, children [WavePlayer_0], properties {pitchMax=0.9, pitchMin=0.8}}
 {id=WavePlayer_0, properties {soundWaveAssetPtr=&/SupplyDropRadio_Landed}}
]
properties {
 attenuationSettings=&Fort_Weapon_Drop_Att
 maxDistance=3000
 soundClassObject=&Fort_Gadgets_Items
 volumeMultiplier=0.25
}
