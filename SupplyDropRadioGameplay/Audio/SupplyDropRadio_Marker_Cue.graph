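# SoundCue for the supply drop radio's marker sound, in the gadget items sound class.
# A short fixed delay leads into a random node that picks one of three equally weighted marker wave variations.
# It borrows a close-range supply drop marker attenuation asset to spatialize playback, with slightly reduced volume.
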SoundCue /SupplyDropRadioGameplay/Audio/SupplyDropRadio_Marker_Cue
root=Delay_0
nodes [
 {id=Delay_0, children [Random_1], properties {delayMax=0.2, delayMin=0.2}}
 {id=Random_1, children [WavePlayer_9, WavePlayer_10, WavePlayer_11], properties {weights [1,1,1]}}
 {id=WavePlayer_10, properties {soundWaveAssetPtr=&/SupplyDropRadio_Marker_02}}
 {id=WavePlayer_11, properties {soundWaveAssetPtr=&/SupplyDropRadio_Marker_03}}
 {id=WavePlayer_9, properties {soundWaveAssetPtr=&/SupplyDropRadio_Marker_01}}
]
properties {
 attenuationSettings=/Game/Sounds/Foley_Loot/Containers/SupplyDrops/Fort_SupplyDrop_Marker_Close_Athena_Att
 maxDistance=1500
 soundClassObject=&Fort_Gadgets_Items
 volumeMultiplier=0.9
}
