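# Sound cue for picking up the Supply Drop Radio item.
# It plays a single wave player with the SupplyDropRadio_Pickup sound, with no randomization or modulation.
# It uses the shared weapon pickup attenuation and is routed to the gadgets and items sound class.
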SoundCue /SupplyDropRadioGameplay/Audio/SupplyDropRadio_Pickup_Cue
root=WavePlayer_0
nodes [{id=WavePlayer_0, properties {soundWaveAssetPtr=&/SupplyDropRadio_Pickup}}]
properties {
 attenuationSettings=&Fort_Weapon_Pickup_Att
 maxDistance=3500
 soundClassObject=&Fort_Gadgets_Items
}
