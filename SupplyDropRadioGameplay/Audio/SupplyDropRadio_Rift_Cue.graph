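# Sound cue for the Supply Drop Radio rift sound, classed under the explosion events sound class.
# A Random node picks one of two equally weighted rift wave variations, which feeds a root Modulator for per-play variation.
# It uses a shared high-frequency supply drop alert attenuation asset and carries over a very long maximum distance.
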SoundCue /SupplyDropRadioGameplay/Audio/SupplyDropRadio_Rift_Cue
root=Modulator_0
nodes [
 {id=Modulator_0, children [Random_0]}
 {id=Random_0, children [WavePlayer_0,WavePlayer_1], properties {weights [1,1]}}
 {id=WavePlayer_0, properties {soundWaveAssetPtr=&/SupplyDropRadio_Rift_01}}
 {id=WavePlayer_1, properties {soundWaveAssetPtr=&/SupplyDropRadio_Rift_02}}
]
properties {
 attenuationSettings=/Game/Sounds/Foley_Loot/Containers/SupplyDrops/Fort_SupplyDrop_Alert_HighFreq_Athena_Att
 maxDistance=230000
 soundClassObject=/Game/Sounds/FortSoundClassesAndModes/Fort_Events_Explosions
}
